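; Thrown Sneaky Snowman projectile that disguises players it hits or spawns a snowman prop where it lands.
; On touch, a scalable-float-gated path applies the snowman effect to eligible pawns or vehicle passengers, then explodes; shallow surfaces disable bouncing.
; On stop without a player hit, the server spawns the prop and sends an actor-limit event.
; BeginPlay removes the empty item unless disguised; explosion plays effects at the server-set hit location.
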(event Custom|PrjSetup
  (bind _instigator (Variables|Actor|GetInstigator)
    (:then
      (bind _asfort_player_pawn (Utilities|Casting|CastToFortPlayerPawn _instigator))
      (bind _output_get (Variables|Default|SetInstigatorPawn _asfort_player_pawn))
      (Variables|Default|SetInstigatorController (Fortnite|GetFortPlayerControllerfromActor _output_get))
      (bind _returnvalue (CallFunction|GetOwnerWeapon))
      (Variables|Default|SetOwnerGUID (FortWeapon|GetInventoryGUID _returnvalue)))
    (:else)))

(event EventBeginPlay
  ;[1 1 1] EVENT -> BeginPlay - If no more charges then remove item
  ;; Updated item removal check
  (CallFunction|PrjSetup)
  (CallFunction|CheckItemInventoryCount))

(event EventOnExploded
  (params (HitActors :type Array%20of%20Actor%20Object%20References) (HitResults :type Array%20of%20Hit%20Result%20Structures))
  (bind _hitlocation (Variables|SneakySnowman|GetHitLocation))
  ;[1 1 1] EVENT -> OnExploded - Audio and clean up
  (Networking|CanExecuteCosmeticEvents
    (:True
      (Effects|SpawnEmitteratLocation (Variables|FX|GetSnowmanSpawnFX) _hitlocation)
      (Camera|PlayWorldCameraShake "/Game/Athena/Items/Consumables/TowerGrenade/FX/B_Ground_CameraShake.B_Ground_CameraShake_C" _hitlocation 25.0 5e+01 2.0)
      (Audio|PlaySoundatLocation (Variables|Sounds|GetSoundOnExplo) _hitlocation)
      (Development|SetHiddeninGame (Variables|Default|GetSK_Sneaky_Snowman_Throw) true false)
      (Components|Activation|Deactivate (Variables|FX|GetFX_Trail))
      (Actor|DestroyActor))))

(event EventOnStop
  (params (Hit :type Hit%20Result%20Structure%20%28by%20ref%29))
  (bind (_returnvalue_roll _returnvalue_pitch _returnvalue_yaw) (Math|Rotator|BreakRotator (Transformation|GetActorRotation (Variables|Actor|GetInstigator))))
  (bind (_bblockinghit _binitialoverlap _time _distance _location _impactpoint _normal _impactnormal _physmat _hitactor) (Collision|BreakHitResult Hit))
  ;[1 1 1] EVENT -> On Stop do Hit FX and Spawn Snowman
  (if (not (Variables|SneakySnowman|GetHitPlayer))
    (CallFunction|SetHitLocation _location)
    (bind _asfort_player_pawn (Utilities|Casting|CastToFortPlayerPawn _hitactor)
      (:then
        (if (CallFunction|ShouldApplySnowman? _asfort_player_pawn)
          _))
      (:CastFailed
        (Networking|SwitchHasAuthority
          (:Authority
            (bind _returnvalue (Building|SpawnBuildingActor "/Figment_Athena_SneakySnowman/Gameplay/Figment_Prop_SneakySnowman.Figment_Prop_SneakySnowman_C" (Math|Transform|MakeTransform _location (Math|Rotator|MakeRotator 0.0 0.0 (+ _returnvalue_yaw -9e+01)) "1.000000,1.000000,1.000000") (Variables|Actor|GetInstigator) (Variables|Actor|GetInstigator)))
            (Ability|SendGameplayEventtoActor (Game|GetInstigator) "(TagName=\"Abilities.Drops.AddToActorLimit.SneakySnowman\")" (Utilities|Struct|MakeGameplayEventData :Target _returnvalue :ContextHandle "()" :InstigatorTags "(GameplayTags=)" :TargetTags "(GameplayTags=)" :EventMagnitude 0.0 :TargetData "()"))))))
    (CallFunction|Explode)))

(event Custom|Explode
  ;[1 1 1] Explode (Custom Event)
  (Delivery|DoExplosion)
  ;;[1 1 1] Stop rotating movement
  (Components|Movement|SetUpdatedComponent (Variables|Default|GetRotatingMovement) 0))

(event Custom|SetHitLocation
  (params (HitLocation :type Vector))
  ;[1 1 1] FORT-1017579; So VFX always run at intended location
  (Networking|SwitchHasAuthority
    (:Authority
      (Variables|SneakySnowman|SetHitLocation HitLocation))))

(event EventOnTouched
  (params (OtherActor :type Actor%20Object%20Reference) (OtherComp :type Primitive%20Component%20Object%20Reference) (HitResult :type Hit%20Result%20Structure%20%28by%20ref%29) (bIsOverlap :type Boolean))
  (bind (_bblockinghit _binitialoverlap _time _distance _location _impactpoint _normal _impactnormal) (Collision|BreakHitResult HitResult))
  ;[1 1 1] EVENT -> OnTouched If we hit a player, apply Snowman to them
  (if (not (Actor|ActorHasTag OtherActor "CreativeVolumeNavInvokerTag"))
    (Variables|SneakySnowman|SetHitActor OtherActor)
    (CallFunction|SetHitLocation _location)
    (if (> (ScalableFloat|GetValueatLevel (Variables|SneakySnowman|GetShouldSnowmanEnemies)) 0.0)
      (bind _asfort_player_pawn (Utilities|Casting|CastToFortPlayerPawn (Variables|SneakySnowman|GetHitActor))
        (:then
          (Networking|SwitchHasAuthority
            (:Authority
              (Utilities|FlowControl|Sequence
                (:then_0
                  (if (CallFunction|ShouldApplySnowman? _asfort_player_pawn)
                    (Environment|ApplyGlobalEnvironmentGameplayEffecttoActor _asfort_player_pawn (Variables|SneakySnowman|GetGE_ApplyEnemySnowman))
                    (Variables|SneakySnowman|SetHitPlayer true)))
                (:then_1
                  (CallFunction|Explode))))))
        (:CastFailed
          (bind (_outactors _returnvalue) (Collision|BoxOverlapActors (Transformation|GetActorLocation) "48.000000,48.000000,200.000000" (Utilities|Array|MakeArray "ObjectTypeQuery3") "/Script/FortniteGame.FortPlayerPawnAthena" :!class /Script/Engine.KismetSystemLibrary))
          (if _returnvalue
            ;; FORT-1017579; Added for case where player hits themselves
            (Networking|SwitchHasAuthority
              (:Authority
                (Variables|SneakySnowman|SetHitPlayer true)
                (CallFunction|Explode)))
            (else
              ;;[1 1 1] If we hit a vehicle
              (bind _asfort_vehicle_interface (Utilities|Casting|CastToFortVehicleInterface (Variables|SneakySnowman|GetHitActor))
                (:then
                  (if (> (Utilities|Array|Length (Vehicle|GetAllPassengers _asfort_vehicle_interface)) 0)
                    (Variables|SneakySnowman|SetHitPlayer true)
                    (Networking|SwitchHasAuthority
                      (:Authority
                        (for _array_element (Vehicle|GetAllPassengers _asfort_vehicle_interface)
                          (if (CallFunction|ShouldApplySnowman? _array_element)
                            (Environment|ApplyGlobalEnvironmentGameplayEffecttoActor _array_element (Variables|SneakySnowman|GetGE_ApplyEnemySnowman))))
                        (CallFunction|Explode)))))
                (:CastFailed
                  _))))))
      (else
        _))
    (CallFunction|ShouldBounce? _impactnormal)))

(event Custom|RemoveGUIDFromInvWithZeroCount
  (params (Item%20GUID :type Guid%20Structure))
  ;[0.9167 0.9167 0.9167 0.5] Remove item with given GUID from player inventory if count is 0
  (bind _instigatorcontroller (Variables|Default|GetInstigatorController)
    (:then
      (bind _returnvalue (Inventory|GetItemNuminStackbyGuid _instigatorcontroller Item%20GUID))
      (if (== _returnvalue 0)
        (bind _returnvalue_2 (Inventory|GetInventoryItemwithGuid _instigatorcontroller Item%20GUID))
        (Inventory|RemoveFortItemFromPlayer _instigatorcontroller _returnvalue_2 -1 true)))
    (:else)))

(event Custom|CheckItemInventoryCount
  ;[1 1 1 0.5] Determine if player is wearing snowman disguise to adjust checks
  (Networking|SwitchHasAuthority
    (:Authority
      (bind _instigatorpawn (Variables|Default|GetInstigatorPawn)
        (:then
          (if (not (GameplayTags|HasMatchingGameplayTag _instigatorpawn "(TagName=\"Athena.Item.Snowman\")"))
            (CallFunction|RemoveGUIDFromInvwithZeroCount (Variables|Default|GetOwnerGUID))))
        (:else)))))

(fn OnRep_HitLocation)

(fn ShouldBounce?
  (params (Hit%20Normal :type Vector))
  ;[1 1 1] Only bounce if we are on a steep enough slope
  (if (> (Math|Vector|DotProduct Hit%20Normal (Math|Vector|GetUpVector "0, 0, 0")) 0.65)
    (Variables|/Script/Engine.ProjectileMovementComponent|SetbShouldBounce false (Variables|Movement|GetProjectileMovementComponent))))

(fn ShouldApplySnowman?
  (params (PlayerPawn :type Object%20Reference))
  (bind _returnvalue (GameplayTags|HasMatchingGameplayTag(Message) PlayerPawn "(TagName=\"Athena.Item.Snowman\")")) ; Make sure they aren't already a snowman
  (return (and (not _returnvalue) (Utilities|NotEqual(Object) PlayerPawn (Game|GetInstigator)))))

(fn UpdateQuest
  (params (FortPC :type Fort%20Player%20Controller%20Object%20Reference))
  ;[1 1 1] Quest hook (if this ever needs to change please let Dave.Ewing or Nicholas.Lacelle know)
  (Networking|SwitchHasAuthority
    (:Authority
      (Quest|SendComplexCustomStatEvent (Quests|GetQuestManager FortPC "Invalid") 0 "(GameplayTags=((TagName=\"Gameplay.Athena.Quests.UseSneakySnowman\")))" "(GameplayTags=)" 1))))

(fn ConstructionScript
  (Networking|CanExecuteCosmeticEvents
    (:True
      (Components|Movement|SetUpdatedComponent (Variables|Default|GetRotatingMovement) (Variables|Default|GetSK_Sneaky_Snowman_Throw))
      (Audio|SetCollisionSoundatIndex (Variables|Default|GetFortCollisionAudio) 0 (Variables|Sounds|GetCue_HitPlayer)))))
